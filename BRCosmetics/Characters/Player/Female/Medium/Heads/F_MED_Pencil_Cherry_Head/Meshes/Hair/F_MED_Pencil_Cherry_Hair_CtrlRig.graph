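# Control Rig for the Pencil Cherry female character's hair that applies a corrective pose to the back hair as the head rotates.
# Each evaluation reads the head's rotation relative to spine_05 and remaps its Euler Z angle to a clamped 0–1 weight.
# That weight drives a chain of Modify Transforms that rotate and shift the left and right hairBack master bones and translate their child 01 bones.
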ControlRigBlueprint /BRCosmetics/Characters/Player/Female/Medium/Heads/F_MED_Pencil_Cherry_Head/Meshes/Hair/F_MED_Pencil_Cherry_Hair_CtrlRig
graphs [
 RigVMFunctionLibrary {links [], nodes []}
 RigVMModel {nodes [
   RigUnit_BeginExecution = &RigUnit_BeginExecution()
   n1 = &RigVMFunction_MathFloatRemap(
    SourceMaximum=-45
    SourceMinimum=10
    TargetMinimum=0
    bClamp=true
    Value=&RigVMFunction_MathQuaternionToEuler(
     RotationOrder=ZYX
     Value=&RigUnit_GetRelativeTransformForItem(Child=(Type=Bone,Name="head"), Parent=(Type=Bone,Name="spine_05")).RelativeTransform.Rotation
    ).Result.Z
   )
   &RigUnit_ModifyTransforms(
    ItemToModify=((Item=(Type=Bone,Name="hairBack_rt_MSTR"),Transform=(Rotation=(X=0,Y=0,Z=0,W=1),Translation=(X=0,Y=1,Z=0),Scale3D=(X=1,Y=1,Z=1))))
    Weight=@n1.Result
    ItemToModify.0.Transform.Rotation=&RigVMFunction_MathQuaternionFromEuler(Euler=(X=0,Y=-42,Z=10)).Result
    ExecutePin=@RigUnit_BeginExecution.ExecutePin
   )
   > &RigUnit_ModifyTransforms(
    ItemToModify=((Item=(Type=Bone,Name="hairBack_lt_MSTR"),Transform=(Rotation=(X=0,Y=0,Z=0,W=1),Translation=(X=1.5,Y=-1,Z=0),Scale3D=(X=1,Y=1,Z=1))))
    Weight=@n1.Result
    ItemToModify.0.Transform.Rotation=&RigVMFunction_MathQuaternionFromEuler(Euler=(X=0,Y=-42,Z=-10)).Result
   )
   > &RigUnit_ModifyTransforms(
    ItemToModify=((Item=(Type=Bone,Name="hairBack_rt_01"),Transform=(Rotation=(X=0,Y=0,Z=0,W=1),Translation=(X=-1,Y=0,Z=2),Scale3D=(X=1,Y=1,Z=1))))
    Weight=@n1.Result
   )
   > &RigUnit_ModifyTransforms(
    ItemToModify=((Item=(Type=Bone,Name="hairBack_lt_01"),Transform=(Rotation=(X=0,Y=0,Z=0,W=1),Translation=(X=-1,Y=-1,Z=2),Scale3D=(X=1,Y=1,Z=1))))
    Weight=@n1.Result
   )
  ]}
]
hierarchy [
 root {loc=<0 0 0>, rot=<0 0 0 1>}
 pelvis {parent=root, loc=<0 -0.5301 98.0571>, rot=<0.0127 0.707 -0.0127 -0.707>}
 spine_01 {parent=pelvis, loc=<6.1208 -0.9328 0>, rot=<0 0 0.034 -0.9994>}
 spine_02 {parent=spine_01, loc=<6.5996 -0.2095 0>, rot=<0 0 0.0238 -0.9997>}
 spine_03 {parent=spine_02, loc=<6.4914 -0.1237 0>, rot=<0 0 0.0149 -0.9999>}
 spine_04 {parent=spine_03, loc=<6.9598 0.0794 0>, rot=<0 0 -0.0132 -0.9999>}
 spine_05 {parent=spine_04, loc=<6.9987 0 0>, rot=<0 0 -0.0015 -1>}
 neck_01 {parent=spine_05, loc=<11.3138 0.1529 0>, rot=<0 0 -0.2003 -0.9797>}
 neck_02 {parent=neck_01, loc=<4.5627 0 0>, rot=<0 0 0.0504 -0.9987>}
 head {parent=neck_02, loc=<5.3914 0 0>, rot=<0 0 0.115 -0.9934>}
 hairBun_lt_MSTR {parent=head, loc=<12.9832 -2.0533 -6>, rot=<0.9486 -0.0341 -0.2892 -0.1237>}
 hairBun_lt_00 {parent=hairBun_lt_MSTR, loc=<0 0 0>, rot=<0 0 0 1>}
 hairSidePiece_rt_MSTR {parent=head, loc=<5.2501 4.5527 7.2567>, rot=<-0.0567 0.0907 0.9941 0.0169>}
 hairSidePiece_rt_00 {parent=hairSidePiece_rt_MSTR, loc=<0 0 0>, rot=<-0.0374 0 0 -0.9993>}
 hairSidePiece_rt_01 {parent=hairSidePiece_rt_00, loc=<4.2832 0 0>, rot=<0.0072 -0.0412 0.0987 -0.9942>}
 hairSidePiece_rt_02 {parent=hairSidePiece_rt_01, loc=<5.1232 0 0>, rot=<0.0374 -0.0395 0.2272 -0.9723>}
 hairSidePiece_lt_MSTR {parent=head, loc=<10.6936 5.347 -4.3802>, rot=<0.107 -0.9586 -0.1837 -0.1893>}
 hairSidePiece_lt_00 {parent=hairSidePiece_lt_MSTR, loc=<0 0 0>, rot=<0.0404 0 0 -0.9992>}
 hairSidePiece_lt_01 {parent=hairSidePiece_lt_00, loc=<6.4886 0 0>, rot=<-0.0432 -0.181 -0.1357 -0.9731>}
 hairSidePiece_lt_02 {parent=hairSidePiece_lt_01, loc=<4.1188 0 0>, rot=<0 0 0 1>}
 hairBun_rt_MSTR {parent=head, loc=<12.9832 -2.0533 6>, rot=<-0.0012 0.2912 -0.004 -0.9567>}
 hairBun_rt_00 {parent=hairBun_rt_MSTR, loc=<0 0 0>, rot=<0 0 0 1>}
 hairBang_MSTR {parent=head, loc=<8.4212 10.7349 0.3193>, rot=<0.0475 -0.9989 0 0>}
 hairBang_00 {parent=hairBang_MSTR, loc=<0 0 0>, rot=<0 0 0 1>}
 hairBack_rt_MSTR {parent=head, loc=<4.8143 -6.009 4.0638>, rot=<-0.0884 -0.4361 0.8946 0.0403>}
 hairBack_rt_00 {parent=hairBack_rt_MSTR, loc=<0 0 0>, rot=<0 0 0 1>}
 hairBack_rt_01 {parent=hairBack_rt_00, loc=<5.5821 0 0>, rot=<0 0.2345 0 -0.9721>}
 hairBack_rt_02 {parent=hairBack_rt_01, loc=<5.3074 1.6206 0>, rot=<0 0 0 1>}
 hairBack_rt_03 {parent=hairBack_rt_01, loc=<4.9287 -3.6597 -1.0513>, rot=<-0.1291 -0.0074 0.0568 -0.99>}
 hairBack_lt_MSTR {parent=head, loc=<4.8146 -6.009 -4.0638>, rot=<-0.0403 -0.8946 0.4361 0.0884>}
 hairBack_lt_00 {parent=hairBack_lt_MSTR, loc=<0 0 0>, rot=<0 0 0 1>}
 hairBack_lt_01 {parent=hairBack_lt_00, loc=<5.5824 0 0.0001>, rot=<0 0.2345 0 -0.9721>}
 hairBack_lt_02 {parent=hairBack_lt_01, loc=<5.3077 -1.6206 -0.0001>, rot=<0 0 0 1>}
 hairBack_lt_03 {parent=hairBack_lt_01, loc=<4.9289 3.6597 -1.0513>, rot=<0.1291 -0.0074 -0.0568 -0.99>}
 clavicle_l {parent=spine_05, loc=<5.4831 -0.1053 -3.2946>, rot=<0.0465 -0.7289 -0.009 -0.6829>}
 upperarm_l {parent=clavicle_l, loc=<9.8658 -1.2136 -0.4105>, rot=<0.0164 -0.3245 -0.0269 -0.9454>}
 upperarm_twist_01_l {parent=upperarm_l, loc=<7.2135 -0.1134 0>, rot=<0 0 0 1>}
 deform_upperarm_fr_base_l {parent=upperarm_twist_01_l, loc=<-5.6312 5.092 0.3354>, rot=<0 0 0 -1>}
 deform_upperarm_fr_l {parent=deform_upperarm_fr_base_l, loc=<0 0 0>, rot=<0 0 0 1>}
 clavicle_r {parent=spine_05, loc=<5.4831 -0.1053 3.2946>, rot=<0.7289 0.0465 -0.6829 0.009>}
 upperarm_r {parent=clavicle_r, loc=<-9.8658 1.2136 0.4105>, rot=<0.0164 -0.3245 -0.0269 -0.9454>}
 upperarm_twist_01_r {parent=upperarm_r, loc=<-7.2135 0.1134 0>, rot=<0 0 0 -1>}
 deform_upperarm_fr_base_r {parent=upperarm_twist_01_r, loc=<5.6312 -5.092 -0.3354>, rot=<0 0 0 -1>}
 deform_upperarm_fr_r {parent=deform_upperarm_fr_base_r, loc=<0 0 0>, rot=<0 0 0 1>}
]
