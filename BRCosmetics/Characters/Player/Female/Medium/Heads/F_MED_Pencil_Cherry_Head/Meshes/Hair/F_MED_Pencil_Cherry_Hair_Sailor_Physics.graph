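# Physics asset for the Sailor hair on the female Pencil Cherry head, defining secondary motion for bangs, buns, side pieces and back hair.
# Capsule bodies cover each hair bone plus head, neck and unconstrained clavicles.
# Each hair group's small MSTR body locks to the head, itself locked to neck_02, anchoring angle-limited chains.
# Back chains fork, with bones 02 and 03 both hanging from 01; left and right limits differ in places.
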PhysicsAsset /BRCosmetics/Characters/Player/Female/Medium/Heads/F_MED_Pencil_Cherry_Head/Meshes/Hair/F_MED_Pencil_Cherry_Hair_Sailor_Physics
bodies {
 clavicle_l [{
   shapeType=Capsule
   length=12
   radius=9
   shapeName=clavicle_l_capsule
   center=<2.0514 0.4492 -4.1963>
   rotation=<rot 80.0111 -178.9221 176.0428>
  }]
 clavicle_r [{
   shapeType=Capsule
   length=12
   radius=9
   shapeName=clavicle_r_capsule
   center=<-2.0514 -0.4492 4.1963>
   rotation=<rot 80.0111 -178.9243 176.0406>
  }]
 hairBack_lt_00 [{shapeType=Capsule, length=7, radius=4, shapeName=hairBack_lt_00_capsule, center=<1 3 0>, rotation=<rot 0 0 90>}]
 hairBack_lt_01 [{shapeType=Capsule, length=7, radius=4, shapeName=hairBack_lt_01_capsule, center=<1 3 1>, rotation=<rot 0 0 90>}]
 hairBack_lt_02 [{shapeType=Capsule, length=7, radius=4, shapeName=hairBack_lt_02_capsule, center=<2 2 1>, rotation=<rot -3.4915 -22.4226 81.0193>}]
 hairBack_lt_03 [{shapeType=Capsule, length=7, radius=4.05, shapeName=hairBack_lt_03_capsule, center=<2 3 2>, rotation=<rot 0 -20 90>}]
 hairBack_lt_MSTR [{shapeType=Capsule, length=0.505, radius=0.505, shapeName=hairBack_lt_MSTR_capsule}]
 hairBack_rt_00 [{shapeType=Capsule, length=7, radius=4, shapeName=hairBack_rt_00_capsule, center=<1 -3 0>, rotation=<rot 0 0 -90>}]
 hairBack_rt_01 [{shapeType=Capsule, length=6.5548, radius=4.4452, shapeName=hairBack_rt_01_capsule, center=<1 -3 1>, rotation=<rot 0 0 -90>}]
 hairBack_rt_02 [{shapeType=Capsule, length=6.6631, radius=4.3369, shapeName=hairBack_rt_02_capsule, center=<2 -2 1>, rotation=<rot -3.4915 22.4226 98.9807>}]
 hairBack_rt_03 [{shapeType=Capsule, length=6.6631, radius=4.3869, shapeName=hairBack_rt_03_capsule, center=<2 -3 2>, rotation=<rot 0 20 90>}]
 hairBack_rt_MSTR [{shapeType=Capsule, length=0.505, radius=0.505, shapeName=hairBack_rt_MSTR_capsule}]
 hairBang_00 [{shapeType=Capsule, length=4.25, radius=4.75, shapeName=hairBang_00_capsule, center=<3.5 2.5 0>}]
 hairBang_MSTR [{shapeType=Capsule, length=0.505, radius=0.505, shapeName=hairBang_MSTR_capsule}]
 hairBun_lt_00 [{shapeType=Capsule, length=0.75, radius=12, shapeName=hairBun_lt_00_capsule, center=<1 0 0>, rotation=<rot 90 0 0>}]
 hairBun_lt_MSTR [{shapeType=Capsule, length=0.505, radius=0.505, shapeName=hairBun_lt_MSTR_capsule, rotation=<rot 0 -90 0>}]
 hairBun_rt_00 [{shapeType=Capsule, length=0.75, radius=12, shapeName=hairBun_rt_00_capsule, center=<1 0 0>, rotation=<rot 90 0 0>}]
 hairBun_rt_MSTR [{shapeType=Capsule, length=0.505, radius=0.505, shapeName=hairBun_rt_MSTR_capsule}]
 hairSidePiece_lt_00 [{shapeType=Capsule, length=3, radius=4, shapeName=hairSidePiece_lt_00_capsule, center=<2 0 2>, rotation=<rot 90 0 0>}]
 hairSidePiece_lt_01 [{shapeType=Capsule, length=3, radius=4, shapeName=hairSidePiece_lt_01_capsule, center=<2 0 3>, rotation=<rot 90 0 0>}]
 hairSidePiece_lt_02 [{shapeType=Capsule, length=3, radius=4, shapeName=hairSidePiece_lt_02_capsule, center=<2 0 3>, rotation=<rot 90 0 0>}]
 hairSidePiece_lt_MSTR [{shapeType=Capsule, length=0.505, radius=0.505, shapeName=hairSidePiece_lt_MSTR_capsule}]
 hairSidePiece_rt_00 [{shapeType=Capsule, length=3, radius=4, shapeName=hairSidePiece_rt_00_capsule, center=<2 0 2>, rotation=<rot 90 0 0>}]
 hairSidePiece_rt_01 [{shapeType=Capsule, length=3, radius=4, shapeName=hairSidePiece_rt_01_capsule, center=<2 0 2>, rotation=<rot 90 0 0>}]
 hairSidePiece_rt_02 [{shapeType=Capsule, length=3, radius=4, shapeName=hairSidePiece_rt_02_capsule, center=<2 0 2>, rotation=<rot 90 0 0>}]
 hairSidePiece_rt_MSTR [{shapeType=Capsule, length=0.505, radius=0.505, shapeName=hairSidePiece_rt_MSTR_capsule}]
 head [{shapeType=Capsule, length=5, radius=6, shapeName=head_capsule, center=<-1.0576 2.1243 0>, rotation=<rot 90 0 0>}]
 neck_02 [{shapeType=Capsule, length=7.75, radius=6.25, shapeName=neck_02_capsule, center=<-5.2446 3.2063 0>, rotation=<rot 90 -166.7373 -166.7372>}]
}
constraints [
 {
  bone1Name=hairBack_lt_MSTR
  bone2Name=head
  swing1LimitDegrees=45
  swing1Motion=Locked
  swing2LimitDegrees=45
  swing2Motion=Locked
  twistLimitDegrees=45
  twistMotion=Locked
 }
 {bone1Name=hairBack_lt_00, bone2Name=hairBack_lt_MSTR, swing1LimitDegrees=1, swing2LimitDegrees=10, twistLimitDegrees=5}
 {bone1Name=hairBack_lt_01, bone2Name=hairBack_lt_00, swing1LimitDegrees=1, swing2LimitDegrees=30, twistLimitDegrees=15}
 {
  bone1Name=hairBack_rt_MSTR
  bone2Name=head
  swing1LimitDegrees=45
  swing1Motion=Locked
  swing2LimitDegrees=45
  swing2Motion=Locked
  twistLimitDegrees=45
  twistMotion=Locked
 }
 {bone1Name=hairBack_rt_00, bone2Name=hairBack_rt_MSTR, swing1LimitDegrees=1, swing2LimitDegrees=10, twistLimitDegrees=5}
 {bone1Name=hairBack_rt_01, bone2Name=hairBack_rt_00, swing1LimitDegrees=1, swing2LimitDegrees=40, twistLimitDegrees=15}
 {
  bone1Name=hairBang_MSTR
  bone2Name=head
  swing1LimitDegrees=45
  swing1Motion=Locked
  swing2LimitDegrees=45
  swing2Motion=Locked
  twistLimitDegrees=45
  twistMotion=Locked
 }
 {bone1Name=hairBang_00, bone2Name=hairBang_MSTR, swing1LimitDegrees=15, swing2LimitDegrees=5, twistLimitDegrees=1}
 {
  bone1Name=hairBun_lt_MSTR
  bone2Name=head
  swing1LimitDegrees=45
  swing1Motion=Locked
  swing2LimitDegrees=45
  swing2Motion=Locked
  twistLimitDegrees=45
  twistMotion=Locked
 }
 {bone1Name=hairBun_lt_00, bone2Name=hairBun_lt_MSTR, swing1LimitDegrees=5, swing2LimitDegrees=5, twistLimitDegrees=1}
 {
  bone1Name=hairBun_rt_MSTR
  bone2Name=head
  swing1LimitDegrees=45
  swing1Motion=Locked
  swing2LimitDegrees=45
  swing2Motion=Locked
  twistLimitDegrees=45
  twistMotion=Locked
 }
 {bone1Name=hairBun_rt_00, bone2Name=hairBun_rt_MSTR, swing1LimitDegrees=5, swing2LimitDegrees=5, twistLimitDegrees=1}
 {
  bone1Name=hairSidePiece_lt_MSTR
  bone2Name=head
  swing1LimitDegrees=45
  swing1Motion=Locked
  swing2LimitDegrees=45
  swing2Motion=Locked
  twistLimitDegrees=45
  twistMotion=Locked
 }
 {bone1Name=hairSidePiece_lt_00, bone2Name=hairSidePiece_lt_MSTR, swing1LimitDegrees=6, swing2LimitDegrees=4, twistLimitDegrees=1}
 {bone1Name=hairSidePiece_lt_01, bone2Name=hairSidePiece_lt_00, swing1LimitDegrees=16, swing2LimitDegrees=3, twistLimitDegrees=1}
 {bone1Name=hairSidePiece_lt_02, bone2Name=hairSidePiece_lt_01, swing1LimitDegrees=16, swing2LimitDegrees=3, twistLimitDegrees=1}
 {
  bone1Name=hairSidePiece_rt_MSTR
  bone2Name=head
  swing1LimitDegrees=45
  swing1Motion=Locked
  swing2LimitDegrees=45
  swing2Motion=Locked
  twistLimitDegrees=45
  twistMotion=Locked
 }
 {bone1Name=hairSidePiece_rt_00, bone2Name=hairSidePiece_rt_MSTR, swing1LimitDegrees=15, swing2LimitDegrees=5, twistLimitDegrees=1}
 {bone1Name=hairSidePiece_rt_01, bone2Name=hairSidePiece_rt_00, swing1LimitDegrees=25, swing2LimitDegrees=5, twistLimitDegrees=1}
 {bone1Name=hairSidePiece_rt_02, bone2Name=hairSidePiece_rt_01, swing1LimitDegrees=25, swing2LimitDegrees=5, twistLimitDegrees=1}
 {
  bone1Name=head
  bone2Name=neck_02
  swing1LimitDegrees=45
  swing1Motion=Locked
  swing2LimitDegrees=45
  swing2Motion=Locked
  twistLimitDegrees=45
  twistMotion=Locked
 }
 {bone1Name=hairBack_rt_02, bone2Name=hairBack_rt_01, swing1LimitDegrees=5, swing2LimitDegrees=30, twistLimitDegrees=20}
 {bone1Name=hairBack_rt_03, bone2Name=hairBack_rt_01, swing1LimitDegrees=20, swing2LimitDegrees=30, twistLimitDegrees=20}
 {bone1Name=hairBack_lt_02, bone2Name=hairBack_lt_01, swing1LimitDegrees=5, swing2LimitDegrees=30, twistLimitDegrees=20}
 {bone1Name=hairBack_lt_03, bone2Name=hairBack_lt_01, swing1LimitDegrees=20, swing2LimitDegrees=30, twistLimitDegrees=20}
]
