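# Anim Blueprint for the Pencil Cherry female head's hair, adding secondary physics motion that adapts to the worn upper-body cosmetic.
# A Control Rig result is cached and feeds two head-space RigidBody simulations with different physics assets.
# Sailor or blazer meta tags select the Sailor simulation, which also offsets the hairBack_rt_MSTR bone.
# The update event also computes Head_Tilt from head-to-spine_05 pitch, which the AnimGraph does not visibly read.
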AnimBlueprint /BRCosmetics/Characters/Player/Female/Medium/Heads/F_MED_Pencil_Cherry_Head/Meshes/Hair/F_MED_Pencil_Cherry_Hair_AnimBP
target_skeleton=/BRCosmetics/Characters/Player/Common/Fortnite_Base_Head/Pencil/Fortnite_Base_Head_Pencil_Skeleton
graphs [
 AnimGraph {nodes [
   Copy Pose From Mesh() {bCopyCurves=true, bUseAttachedParent=true}
   n1 = RigidBody(ComponentPose=Local To Component(LocalPose=Use cached pose 'SavedPose14224'())) {
    bOverrideWorldGravity=true
    overridePhysicsAsset=&/F_MED_Pencil_Cherry_Hair_Sailor_Physics
    simulationSpace=BaseBoneSpace
    alphaBoolBlend {blendOption=Linear}
    baseBoneRef {boneName=head}
    componentAppliedLinearAccClamp=<3000 3000 3000>
    componentLinearAccScale=<0.48 0.2 0.2>
    componentLinearVelScale=<0.2 0.1 0.1>
    overrideWorldGravity {z=-980}
    simSpaceSettings {
     maxAngularAcceleration=3000
     maxAngularVelocity=3
     maxLinearAcceleration=700
     maxLinearVelocity=500
     velocityScaleZ=0
     worldAlpha=1
     externalLinearDragV=<0.2 0.2>
    }
   }
   n2 = "Transform (Modify) Bone - Bone: hairBack_rt_MSTR"(ComponentPose=@n1, Alpha=1, Translation="-1,0,0", Rotation="3,0,0", Scale="1,1,1") {
    rotationMode=BMM_Additive
    rotationSpace=BCS_BoneSpace
    translationMode=BMM_Additive
    translationSpace=BCS_BoneSpace
    alphaBoolBlend {blendOption=Linear}
    boneToModify {boneName=hairBack_rt_MSTR}
    rotation {pitch=3}
    translation {x=-1}
   }
   n3 = RigidBody(ComponentPose=Local To Component(LocalPose=Use cached pose 'SavedPose14224'())) {
    bOverrideWorldGravity=true
    overridePhysicsAsset=&/F_MED_Pencil_Cherry_Hair_Physics
    simulationSpace=BaseBoneSpace
    alphaBoolBlend {blendOption=Linear}
    baseBoneRef {boneName=head}
    componentAppliedLinearAccClamp=<3000 3000 3000>
    componentLinearAccScale=<0.48 0.2 0.2>
    componentLinearVelScale=<0.2 0.1 0.1>
    overrideWorldGravity {z=-980}
    simSpaceSettings {
     maxAngularAcceleration=3000
     maxAngularVelocity=3
     maxLinearAcceleration=700
     maxLinearVelocity=500
     velocityScaleZ=0
     worldAlpha=1
     externalLinearDragV=<0.2 0.2>
    }
   }
   Output Pose(Result=Blend Poses by bool(
     BlendPose_0=Component To Local(ComponentPose=@n2)
     BlendPose_1=Component To Local(ComponentPose=@n3)
     bActiveValue=Get Blazer/Sailor Jacket() -> (Blazer/Sailor Jacket:bool)
    ) {blendType=Linear}) {layerGroup=None}
   n4 = Control Rig(Source=Input Pose()) {
    alphaBoolBlend {blendOption=Linear}
    controlRigAssetReference {blueprintRigClass=&/F_MED_Pencil_Cherry_Hair_CtrlRig.F_MED_Pencil_Cherry_Hair_CtrlRig_C}
   }
   Save cached pose 'SavedPose14224'(Pose=@n4) {cacheName=SavedPose14224}
  ]}
 EventGraph {nodes [
   Event BlueprintUpdateAnimation()
   > Set Blazer/Sailor Jacket(Blazer/Sailor Jacket:bool=CharacterItemDefinitionHasAnyMetaTag(Tags=Make Array!(
      "[0]"=(TagName="Cosmetics.MetaTags.Pencil.Upper.Sailor")
      "[1]":Gameplay Tag Structure=(TagName="Cosmetics.MetaTags.Pencil.Upper.BlazerCuff")
      "[2]":Gameplay Tag Structure=(TagName="Cosmetics.MetaTags.Pencil.Upper.BlazerOpen")
     ) -> (Array))) -> (Output_Get:bool)
   n1 = BreakTransform(InTransform=GetDeltaTransformFromRefPose(self=GetOwnerSkeletalMeshComponent(), BoneName=head, BaseName=spine_05))
   > Set Head_Tilt(Head_Tilt:double=MapRangeClamped(Value=BreakRotator(InRot=@n1.Rotation).Pitch, InRangeA=30, InRangeB=45)) -> (Output_Get:double)
  ]}
]
variables [
 Blazer/Sailor Jacket {type=boolean}
 Head_Tilt {type=number}
]
